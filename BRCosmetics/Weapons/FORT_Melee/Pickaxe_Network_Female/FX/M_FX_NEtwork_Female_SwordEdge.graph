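# Translucent unlit material for Niagara mesh particles, apparently the sword-edge effect on the female Network pickaxe.
# Particle color RGB feeds an emissive adjustments function and, separately, a multiply by vertex color red.
# A second multiply combines particle alpha with vertex alpha, so both per-particle values and painted mesh data can mask the effect.
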Material /BRCosmetics/Weapons/FORT_Melee/Pickaxe_Network_Female/FX/M_FX_NEtwork_Female_SwordEdge
graph {
 properties {
  bUsedWithNiagaraMeshParticles=true
  blendMode=BLEND_Translucent
  pixelDepthOffsetMode=PDOM_Legacy
  previewMesh=&/SM_FX_Network_Female
  shadingModel=MSM_Unlit
  usedShadingModels=Unlit
 }
 nodes [
  n1 = ParticleColor()
  MaterialFunctionCall(Emissive=@n1.RGB) {materialFunction=&MF_FX_EmissiveAdjustments}
  n2 = VertexColor()
  Multiply(A=@n1.A, B=@n2.A)
  Multiply(A=@n1.RGB, B=@n2.R)
 ]
}
