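# Translucent unlit Niagara sprite and ribbon material that draws glowing hexagons for the Network Female pickaxe effects.
# A static switch picks tiled hex wrap textures or a procedural UV hexagon ring shaped by dynamic parameters.
# Particle color tints the mask, scaled by ColorScale, plus a squared view-to-normal term in texture mode, through a shared emissive function.
# The IsSwingTrail switch fades the mask along second-channel UVs.
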Material /BRCosmetics/Weapons/FORT_Melee/Pickaxe_Network_Female/FX/M_Hexagon_NetworkFemale
graph {
 properties {
  bUsedWithNiagaraRibbons=true
  bUsedWithNiagaraSprites=true
  blendMode=BLEND_Translucent
  pixelDepthOffsetMode=PDOM_Legacy
  shadingModel=MSM_Unlit
  usedShadingModels=Unlit
 }
 nodes [
  n1 = TextureCoordinate() {coordinateIndex=1}
  ComponentMask(None=@n1) {r=true}
  n2 = Multiply(A=@n1, B=@n1)
  ComponentMask(None=@n2) {g=true}
  n3 = StaticBoolParameter() {parameterName=Use Texture}
  n4 = DotProduct(
   A=CameraVectorWS()
   B=StaticSwitch(
    True=TextureSample() {samplerType=SAMPLERTYPE_Normal, texture=/Game/Weapons/WeaponSkins/Wraps/Textures/Hex/T_Wrap_Hex_N}.RGB
    False=Constant3Vector()
    Value=@n3
   )
  )
  n5 = Multiply(A=@n4, B=@n4)
  n6 = ParticleColor()
  n7 = DynamicParameter() {parameterIndex=1, defaultValue=<rgba 0.46 1.2 0.6666 4>, paramNames [Bold, U Size, V Size, ColorScale]}
  n8 = VectorParameter() {parameterName=UV Scale, defaultValue=<rgba 1 1 0 0>}
  n9 = AppendVector(A=@n8.R, B=@n8.R)
  n10 = TextureCoordinate()
  n11 = OneMinus(None=TextureSample(UVs=Multiply(A=@n9, B=@n10)) {texture=/Game/Weapons/WeaponSkins/Wraps/Textures/Hex/T_Wrap_Hex_BC}.G)
  n12 = Multiply(A=@n11, B=@n11)
  n13 = TextureCoordinate()
  n14 = MaterialFunctionCall(Float2=Abs(None=Subtract(A=@n13) {constB=0.5})) {materialFunction=&BreakOutFloat2Components}
  n15 = DynamicParameter() {parameterIndex=1, defaultValue=<rgba 0.46 1.2 0.6666 4>, paramNames [Bold, U Size, V Size, ColorScale]}
  n16 = Max(A=Multiply(A=@n14.R, B=@n15.U Size) {constB=1.2}, B=Add(A=@n14.G, B=Multiply(A=@n14.R, B=@n15.V Size) {constB=0.666}))
  n17 = Add(A=@n16, B=@n15.Bold) {constB=0.45}
  n18 = OneMinus(None=Saturate(None=Multiply(A=@n17, B=@n17)))
  n19 = StaticSwitch(True=Multiply(A=@n12, B=@n12), False=Multiply(A=Step(Y=@n16) {constX=0.48}, B=Step(Y=@n18) {constX=0.2}), Value=@n3)
  n20 = Multiply(A=@n6.RGB, B=@n19)
  MaterialFunctionCall(Emissive=Add(A=Multiply(A=@n5, B=@n6.RGB), B=Multiply(A=@n7.ColorScale, B=@n20))) {materialFunction=&MF_FX_EmissiveAdjustments}
  n21 = OneMinus(None=@n2)
  n22 = ComponentMask(None=Multiply(A=@n21, B=@n21)) {r=true}
  n23 = ComponentMask(None=@n21) {g=true}
  n24 = Multiply(A=Multiply(A=Multiply(A=@n19, B=@n22), B=Multiply(A=@n23, B=@n2)))
  StaticSwitch(True=Multiply(A=@n24, B=@n24), False=@n19, Value=StaticBoolParameter() {parameterName=IsSwingTrail})
 ]
}
