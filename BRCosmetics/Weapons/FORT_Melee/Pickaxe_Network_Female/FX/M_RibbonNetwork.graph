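# Unlit Niagara ribbon material in the effects folder of the female Network pickaxe.
# An HDR Color parameter multiplies particle color and feeds the MF_FX_EmissiveAdjustments function.
# Screen-space vertical UV, repeated through Frac and Step, builds fine horizontal scanlines that alternate particle alpha between half and full strength before it is squared.
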Material /BRCosmetics/Weapons/FORT_Melee/Pickaxe_Network_Female/FX/M_RibbonNetwork
graph {
 properties {bUsedWithNiagaraRibbons=true, pixelDepthOffsetMode=PDOM_Legacy, shadingModel=MSM_Unlit, usedShadingModels=Unlit}
 nodes [
  n1 = ParticleColor()
  MaterialFunctionCall(Emissive=Multiply(A=VectorParameter() {parameterName=Color, defaultValue=<rgba 10 10 10 0>}.RGB, B=@n1.RGB)) {materialFunction=&MF_FX_EmissiveAdjustments}
  n2 = ComponentMask(None=ScreenPosition().ViewportUV) {g=true}
  n3 = Step(Y=Frac(None=Multiply(A=@n2) {constB=256})) {constX=0.5}
  n4 = Multiply(A=Add(A=@n1.A, B=Multiply(A=@n1.A, B=@n3))) {constB=0.5}
  Multiply(A=@n4, B=@n4)
 ]
}
