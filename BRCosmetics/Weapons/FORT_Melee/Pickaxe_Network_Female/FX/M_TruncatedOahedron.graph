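# Unlit, two-sided translucent material for Niagara mesh particles in the Pickaxe_Network_Female melee cosmetic effects.
# Vertex color and dynamic-parameter channel sums mask particle color, with a V Scroll threshold on inverted UV V gating a hard-edged wipe.
# That result is added to particle color and fed into MF_FX_EmissiveAdjustments.
# A separate branch builds screen-space horizontal scanlines from viewport V, modulated by particle alpha.
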Material /BRCosmetics/Weapons/FORT_Melee/Pickaxe_Network_Female/FX/M_TruncatedOahedron
graph {
 properties {
  bUsedWithNiagaraMeshParticles=true
  blendMode=BLEND_Translucent
  pixelDepthOffsetMode=PDOM_Legacy
  shadingModel=MSM_Unlit
  twoSided=true
  usedShadingModels=Unlit
 }
 nodes [
  n1 = ParticleColor()
  n2 = VertexColor()
  n3 = Saturate(None=Add(A=Add(A=@n2.R, B=@n2.R), B=@n2.B))
  n4 = DynamicParameter() {defaultValue=<rgba 1 0 0 0>, paramNames [R, G, B, V Scroll]}
  n5 = Add(A=Add(A=@n4.R, B=@n4.R), B=@n4.B)
  n6 = TextureCoordinate()
  n7 = Add(A=@n4.V Scroll, B=OneMinus(None=ComponentMask(None=@n6) {g=true}))
  n8 = Constant()
  n9 = If(A=@n7, B=Constant() {r=0.99}, "A > B"=@n8, "A == B"=@n8, "A < B"=Ceil(None=@n7))
  n10 = Multiply(A=Multiply(A=@n1.RGB, B=Multiply(A=@n3, B=@n5)), B=Add(A=Saturate(None=@n9)) {constB=0.01})
  n11 = ComponentMask(None=ScreenPosition().ViewportUV) {g=true}
  n12 = Step(Y=Frac(None=Multiply(A=@n11) {constB=256})) {constX=0.5}
  n13 = Multiply(A=Add(A=@n1.A, B=Multiply(A=@n1.A, B=@n12))) {constB=0.5}
  Add(A=@n10, B=Multiply(A=@n13, B=@n13))
  n14 = Frac(None=Time())
  Add(A=Multiply(A=Subtract(A=@n14) {constB=0.5}) {constB=3}) {constB=0.5}
  MaterialFunctionCall(Emissive=Add(A=@n1.RGB, B=@n10)) {materialFunction=&MF_FX_EmissiveAdjustments}
 ]
}
